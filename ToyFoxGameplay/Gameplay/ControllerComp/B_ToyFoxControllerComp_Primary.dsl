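; Primary player controller component for ToyFox that cleans up game state when its player disconnects.
; It checks the event state on the game state component; one specific state, apparently the intro, runs an intro exit using the controlled pawn.
; Otherwise it calls SkipDestroySelfonSwitchBack on the giant player pawn, runs the pilot exit with that pawn and clears all power levels.
; BeginPlay and Tick are empty.
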(event EventBeginPlay)

(event EventTick
  (params (DeltaSeconds :type Float%20%28single-precision%29)))

(event EventHandlePlayerDisconnecting
  (bind _gamestatecomp (Variables|Default|GetGameStateComp))
  (bind _pawn (Giant|GetGiantPawn _gamestatecomp))
  (if (Utilities|Enum|Equal(Enum) (Event|GetEventState (Variables|Default|GetGameStateComp)) "NewEnumerator2")
    (bind _humanpawn (CallFunction|GetControlledPawn))
    (Pilots|HandlePlayerExitforIntro (Variables|Default|GetGameStateComp) _humanpawn)
    (else
      (bind _astoy_fox_giant_player_pawn (Utilities|Casting|CastToToyFoxGiantPlayerPawn _pawn))
      (Class|ToyFoxGiantPlayerPawn|SkipDestroySelfonSwitchBack _astoy_fox_giant_player_pawn)
      (Pilots|HandlePlayerExit _gamestatecomp _pawn)
      (PowerLevel|ClearAllPowerLevel _gamestatecomp))))
