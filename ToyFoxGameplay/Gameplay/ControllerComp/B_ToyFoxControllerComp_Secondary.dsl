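; Secondary ToyFox player controller component that reports a disconnecting player to the ToyFox game state component.
; It caches that component from the Athena game state on BeginPlay.
; On disconnect, it passes the owning controller's Fort player pawn to either the intro exit handler or the standard exit handler, depending on the current event state.
; The Tick event is empty.
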(event EventBeginPlay
  (Variables|Default|SetGameStateComp (Actor|GetComponentbyClass (Fortnite|Athena|GameState|GetGameStateAthena) "/ToyFoxGameplay/Gameplay/B_ToyFoxGameStateComponent.B_ToyFoxGameStateComponent_C")))

(event EventTick
  (params (DeltaSeconds :type Float%20%28single-precision%29)))

(event EventHandlePlayerDisconnecting
  (if (Utilities|Enum|Equal(Enum) (Event|GetEventState (Variables|Default|GetGameStateComp)) "NewEnumerator2")
    (bind _humanpawn (CallFunction|GetControlledPawn))
    (Pilots|HandlePlayerExitforIntro (Variables|Default|GetGameStateComp) _humanpawn)
    (else
      (bind _humanpawn_2 (CallFunction|GetControlledPawn))
      (Pilots|HandlePlayerExit (Variables|Default|GetGameStateComp) _humanpawn_2))))

(fn GetControlledPawn
  (bind _asfort_player_controller (Utilities|Casting|CastToFortPlayerController (Components|GetOwner)))
  (bind _asfort_player_pawn (Utilities|Casting|CastToFortPlayerPawn (Pawn|GetControlledPawn _asfort_player_controller)))
  (return _asfort_player_pawn))
