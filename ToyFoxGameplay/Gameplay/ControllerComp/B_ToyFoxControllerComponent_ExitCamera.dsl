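; Short-lived player controller component that turns the camera to a set rotation after the pawn is teleported.
; BeginPlay ignores look input, and the server starts a timer that soon destroys the component.
; When LookAtRotation replicates, clients that can run cosmetic events wait briefly for the teleport to finish, then set the control rotation.
; EndPlay restores look input.
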(event EventBeginPlay
  ;[0 0.15 0.04704 0.5] BeginPlay: Ignore look input and start a timer (server) to destroy the component after a short duration.
  (|CallParentFunction)
  (Input|SetIgnoreLookInput (Utilities|Casting|CastToFortPlayerController (Components|GetOwner)) true)
  (Networking|SwitchHasAuthority
    (:Authority
      (Utilities|Time|SetTimerbyEvent (delegate Custom|DestroyComponent) 0.66))))

(event Custom|DestroyComponent
  ;[0 0.15 0.04704 0.5] BeginPlay: Ignore look input and start a timer (server) to destroy the component after a short duration.
  (Components|DestroyComponent))

(event Custom|OnRep_LookAtRotation
  ;[0.1422 0 0.15 0.5] OnRep_LookAtRotation: After a short duration (pawn teleport to new location needs to be completed), update the control rotation.
  (Networking|CanExecuteCosmeticEvents
    (:True
      (Utilities|FlowControl|Delay 0.33)
      (bind _asfort_player_controller (Utilities|Casting|CastToFortPlayerController (Components|GetOwner)))
      (Fortnite|KeepCGPathOptimizations _asfort_player_controller 0 1.0)
      (Pawn|SetControlRotation _asfort_player_controller (Variables|Default|GetLookAtRotation)))))

(event EventEndPlay
  (params (EndPlayReason :type EEndPlayReason%20Enum :default "Destroyed"))
  ;[0.09375 0.04155 0 0.5] EndPlay: Allow camera look movement again.
  (Utilities|FlowControl|Sequence
    (:then_0
      (Input|SetIgnoreLookInput (Utilities|Casting|CastToFortPlayerController (Components|GetOwner)) false))
    (:then_1
      (|CallParentFunction EndPlayReason))))
